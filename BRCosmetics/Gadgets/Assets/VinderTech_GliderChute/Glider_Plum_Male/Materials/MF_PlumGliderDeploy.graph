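# Material function for the Plum glider deploy effect, driving a dissolve through Opacity Mask and Emissive on incoming material attributes.
# Plum_Transition scales an origin-centered sphere mask on pre-skinned position, with panning noise distorting its radius and plum noise eroding it.
# A second mask adds a Fresnel-weighted edge glow.
# Vertex-red areas stay opaque and blink on a sine (Claw_BlinkCount); a hash-noise branch is left unconnected.
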MaterialFunction /BRCosmetics/Gadgets/Assets/VinderTech_GliderChute/Glider_Plum_Male/Materials/MF_PlumGliderDeploy
graph {nodes [
  Constant2Vector()
  n1 = GetMaterialAttributes(None=FunctionInput() {inputName=Material, inputType=FunctionInput_MaterialAttributes}) {attributeGetTypes [#1]}
  n2 = VertexColor()
  n3 = VectorParameter() {parameterName=LX_TransitionEdge_Color, defaultValue=<rgba 0.4382 0 1 1>}
  n4 = ScalarParameter() {defaultValue=1, parameterName=FresnelExponent_02}
  n5 = PixelNormalWS()
  n6 = LinearInterpolate(
   A=ScalarParameter() {defaultValue=1, parameterName=LX_TransitionEdge_MinGlowStrength}
   B=ScalarParameter() {defaultValue=10, parameterName=LX_TransitionEdge_MaxGlowStrength}
   Alpha=Power(Base=Fresnel(ExponentIn=@n4, Normal=@n5), Exp=ScalarParameter() {defaultValue=1, parameterName=PlumFresnelPower}) {constExponent=1.5}
  )
  n7 = Multiply(A=Multiply(A=@n2.R, B=@n3.RGB), B=@n6)
  n8 = Time()
  n9 = ScalarParameter() {defaultValue=2, parameterName=Claw_BlinkCount}
  n10 = Sine(None=Add(A=Multiply(A=@n8, B=@n9) {constB=2}) {constB=0.25})
  n11 = PreSkinnedPosition()
  n12 = ShaderStageSwitch(PixelShader=VertexInterpolator(VS=@n11).PS, VertexShader=@n11)
  n13 = VectorParameter() {parameterName=LX_OffsetNoiseTiling, defaultValue=<rgba 2 2 1 1>}
  n14 = Panner(Coordinate=Multiply(A=TextureCoordinate(), B=ComponentMask(None=@n13.RGB) {g=true, r=true})) {speedX=0.25}
  n15 = Add(
   A=ScalarParameter() {defaultValue=256, parameterName=LX_MaskRadius}
   B=Multiply(
    A=TextureSample(UVs=@n14) {texture=/Game/Characters/Player/Male/Medium/Bodies/M_MED_Viper/Textures/Glitch/T_SoftNoise}.G
    B=ScalarParameter() {defaultValue=256, parameterName=LX_EdgeLineOffset_Amt}
   )
  )
  n16 = ScalarParameter() {parameterName=Plum_Transition}
  n17 = Multiply(
   A=TextureCoordinate()
   B=ComponentMask(None=VectorParameter() {parameterName=LX_TransitionNoiseTiling, defaultValue=<rgba 10 10 10 1>}.RGB) {g=true, r=true}
  )
  n18 = Subtract(
   A=SphereMask(
    A=FunctionInput(Preview=@n12) {bUsePreviewValueAsDefault=true, inputName=PreskinnedPosition}
    B=Constant3Vector()
    Radius=Multiply(A=@n15, B=@n16)
    Hardness=ScalarParameter() {parameterName=LX_MaskHardness}
   )
   B=Multiply(
    A=TextureSample(UVs=@n17) {samplerType=SAMPLERTYPE_Masks, texture=/BRCosmetics/Gadgets/Assets/VinderTech_GliderChute/Glider_Plum_Male/Textures/T_PlumNoise}.A
    B=ScalarParameter() {defaultValue=0.5, parameterName=LX_TransitionNoiseStrength}
   )
  )
  n19 = SphereMask(
   A=@n18
   B=Constant()
   Radius=ScalarParameter() {defaultValue=0.2, parameterName=LX_TransitionEdgeRadius}
   Hardness=ScalarParameter() {parameterName=LX_TransitionEdgeHardness}
  )
  n20 = Multiply(A=Multiply(A=@n19, B=@n3.RGB), B=@n6)
  n21 = Add(A=Multiply(A=Power(Base=@n7), B=ConstantBiasScale(None=@n10)), B=LinearInterpolate(A=Power(Base=@n20), B=@n1.EmissiveColor, Alpha=@n2.R))
  n22 = Add(A=Ceil(None=@n18), B=@n2.R)
  FunctionOutput(None=SetMaterialAttributes!(MaterialAttributes=@n1.MaterialAttributes, EmissiveColor=Add(A=@n1.EmissiveColor, B=@n21), OpacityMask=Clamp(None=@n22)) {attributeSetTypes [#1, #2]})
  MaterialFunctionCall() {materialFunction=&ObjectPivotPoint}
  n23 = TextureCoordinate()
  n24 = Constant() {r=96}
  n25 = MaterialFunctionCall(Seed=Multiply(A=Time()), Vec2=Ceil(None=Multiply(A=@n23, B=@n24))) {materialFunction=&MF_Hash}
  Multiply(A=@n19, B=ConstantBiasScale(None=StaticComponentMaskParameter(None=@n25.Vec2) {defaultR=true}))
  ObjectPositionWS()
 ]}
